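# Small pip widget in the Survivor Tiers quest rewards UI that shows a prestige state as complete or incomplete.
# A fixed-size SizeBox holds one slightly rotated Image, which shows a circle material by default.
# SetComplete takes a bool and swaps the Image brush between the CompleteState and IncompleteState pip textures stored as brush variables.
# The event graph is empty, so the state changes only when SetComplete is called.
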WidgetBlueprint /Game/UI/SurvivorTiers/QuestRewards/SurvivorPanelPrestigeState
graphs [
 EventGraph {links [], nodes []}
 SetComplete {nodes [
   n1 = SetComplete!() -> (then, IsComplete:bool=~) {is=title K2Node_FunctionEntry}
   > SetBrush(
    self=Get Image_State() -> (Image_State:Image Object Reference)
    InBrush=Select(
     Option 0:"Slate Brush Structure (by ref)"=Get IncompleteState() -> (IncompleteState:Slate Brush Structure)
     Option 1:"Slate Brush Structure (by ref)"=Get CompleteState() -> (CompleteState:Slate Brush Structure)
     Index=@n1
    ) -> (ReturnValue:"Slate Brush Structure (by ref)")
   )
  ]}
]
variables [
 CompleteState {
  type=SlateBrush
  default {
   bIsDynamicallyLoaded=false
   drawAs=Image
   imageType=NoImage
   mirroring=NoMirror
   resourceName=None
   resourceObject=/Game/UI/SurvivorTiers/QuestRewards/Textures/CirclePip_Complete
   tiling=NoTile
   imageSize=<32 32>
   margin=<ltrb 0 0 0 0>
   outlineSettings {
    bUseBrushTransparency=false
    roundingType=HalfHeightRadius
    width=0
    color {colorUseRule=UseColor_Specified, specifiedColor=<rgba 0 0 0 0>}
    cornerRadii=<0 0 0 1>
   }
   tintColor {colorUseRule=UseColor_Specified, specifiedColor=<rgba 1 1 1 1>}
   uVRegion {bIsValid=false, max=<0 0>, min=<0 0>}
  }
 }
 IncompleteState {
  type=SlateBrush
  default {
   bIsDynamicallyLoaded=false
   drawAs=Image
   imageType=NoImage
   mirroring=NoMirror
   resourceName=None
   resourceObject=/Game/UI/SurvivorTiers/QuestRewards/Textures/CirclePip_Incomplete
   tiling=NoTile
   imageSize=<32 32>
   margin=<ltrb 0 0 0 0>
   outlineSettings {
    bUseBrushTransparency=false
    roundingType=HalfHeightRadius
    width=0
    color {colorUseRule=UseColor_Specified, specifiedColor=<rgba 0 0 0 0>}
    cornerRadii=<0 0 0 1>
   }
   tintColor {colorUseRule=UseColor_Specified, specifiedColor=<rgba 1 1 1 1>}
   uVRegion {bIsValid=false, max=<0 0>, min=<0 0>}
  }
 }
]
widgets {
 info {inheritedWidgetCount=0, parentClass=&UserWidget, rootWidgetClass=&SizeBox, widgetCount=2}
 widgets [
  SizeBox_0 {
   widgetClassPath=&SizeBox
   widget=SizeBox_0
   properties {bOverride_HeightOverride=true, bOverride_WidthOverride=true, heightOverride=32, widthOverride=32}
  }
  Image_State {
   widgetClassPath=&Image
   bIsVariable=true
   parent=SizeBox_0
   widget=Image_State
   properties {
    brush {
     bIsDynamicallyLoaded=false
     drawAs=Image
     imageType=NoImage
     mirroring=NoMirror
     resourceName=None
     resourceObject=/Game/Athena/UI/Materials/M_UI_Circle
     tiling=NoTile
     imageSize=<32 32>
     margin=<ltrb 0 0 0 0>
     outlineSettings {
      bUseBrushTransparency=false
      roundingType=HalfHeightRadius
      width=0
      color {colorUseRule=UseColor_Specified, specifiedColor=<rgba 0 0 0 0>}
      cornerRadii=<0 0 0 0>
     }
     tintColor {colorUseRule=UseColor_Specified, specifiedColor=<rgba 1 1 1 1>}
     uVRegion {bIsValid=false, max=<0 0>, min=<0 0>}
    }
    renderTransform {angle=-10, scale=<1 1>, shear=<0 0>, translation=<0 0>}
   }
   slot {class=SizeBoxSlot}
  }
 ]
}
