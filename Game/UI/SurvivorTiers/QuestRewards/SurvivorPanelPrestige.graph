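# Widget for one prestige level in the survivor tiers quest rewards UI, showing an icon, progress bar and completion state.
# PreConstruct uses PresigeLevel to read seasonal display data for title and lazy-loaded icon, sizes the line from LineHeight and toggles the bar via ShowProgress.
# SetProgress fills the bar from PercentToNext and passes Achieved to the state child widget.
# Title and line are hidden by default, and the graph never reveals them.
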WidgetBlueprint /Game/UI/SurvivorTiers/QuestRewards/SurvivorPanelPrestige
graphs [
 EventGraph {nodes [
   Event PreConstruct()
   n1 = Break Survivor Presige Data!(SurvivorPresigeData:"Survivor Presige Data Structure (by ref)"=GetPrestigeData!(
     self:Athena Seasonal Survivor Tiers Display Data Object Reference=GetSurvivorTierDisplayData!() -> (ReturnValue:Athena Seasonal Survivor Tiers Display Data Object Reference)
     PrestigeLevel:int=Get PresigeLevel() -> (PresigeLevel:int=0)
    ) -> (ReturnValue:Survivor Presige Data Structure)) -> (DisplayName:Text, SmallIconMaterial:Material Interface Soft Object Reference, TextColor:Linear Color Structure) {is=title K2Node_BreakStruct}
   > "SetText (Text)"(self=Get Text_Title() -> (Text_Title:Common Text Object Reference), InText=@n1)
   > SetBrushFromLazyMaterial!(
    execute
    self:Common Lazy Image Object Reference=Get LazyImage_Icon() -> (LazyImage_Icon:Fort Lazy Image Object Reference)
    LazyMaterial:"Material Interface Soft Object Reference (by ref)"=@n1.SmallIconMaterial
   ) -> (then)
   n2 = Get Image_Line() -> (Image_Line:Image Object Reference)
   > n3 = Set members in Slate Brush!(
    execute:Exec
    StructRef:"Slate Brush Structure (by ref)"=Get Brush!(self:Image Object Reference=@n2) -> (Brush:Slate Brush Structure)
    ImageSize:Vector 2D Structure=MakeVector2D(X=2, Y=Get LineHeight() -> (LineHeight:double))
   ) -> (then:Exec, StructOut:"Slate Brush Structure (by ref)") {is=title K2Node_SetFieldsInStruct}
   > SetBrush(self=@n2, InBrush=@n3)
   > SetVisibility(
    self=Get ProgressBar_Progress() -> (ProgressBar_Progress:Progress Bar Object Reference)
    InVisibility=Select(Option 0=Collapsed, Option 1=HitTestInvisible, Index=Get ShowProgress() -> (ShowProgress:bool))
   )
  ]}
 SetProgress {nodes [
   n1 = SetProgress!() -> (then, PercentToNext:double=~, Achieved:bool=~) {is=title K2Node_FunctionEntry}
   > SetPercent!(execute, self:Progress Bar Object Reference=Get ProgressBar_Progress(), InPercent:float=@n1) -> (then)
   > SetComplete!(
    execute
    self:Survivor Panel Prestige State Object Reference=Get Widget_State() -> (Widget_State:Survivor Panel Prestige State Object Reference)
    IsComplete:bool=@n1.Achieved
   ) -> (then)
  ]}
]
variables [
 LineHeight {type=number, default=32}
 PresigeLevel {type=integer}
 ShowProgress {type=boolean, default=true}
]
widgets {
 info {inheritedWidgetCount=0, parentClass=&UserWidget, rootWidgetClass=&VerticalBox, widgetCount=9}
 widgets [
  VerticalBox_7 {widgetClassPath=&VerticalBox, widget=VerticalBox_7}
  Overlay_1 {
   widgetClassPath=&Overlay
   parent=VerticalBox_7
   widget=Overlay_1
   properties {renderTransform {angle=0, scale=<1 1>, shear=<0 0>, translation=<0 10>}}
   slot {class=VerticalBoxSlot, properties {horizontalAlignment=HAlign_Center, verticalAlignment=VAlign_Center}}
  }
  LazyImage_Icon {
   widgetClassPath=&FortLazyImage
   bIsVariable=true
   parent=Overlay_1
   widget=LazyImage_Icon
   properties {brush {
     bIsDynamicallyLoaded=false
     drawAs=Image
     imageType=NoImage
     mirroring=NoMirror
     resourceName=None
     resourceObject=null
     tiling=NoTile
     imageSize=<100 100>
     margin=<ltrb 0 0 0 0>
     outlineSettings {
      bUseBrushTransparency=false
      roundingType=HalfHeightRadius
      width=0
      color {colorUseRule=UseColor_Specified, specifiedColor=<rgba 0 0 0 0>}
      cornerRadii=<0 0 0 0>
     }
     tintColor {colorUseRule=UseColor_Specified, specifiedColor=<rgba 1 1 1 1>}
     uVRegion {bIsValid=false, max=<0 0>, min=<0 0>}
    }}
   slot {class=OverlaySlot, properties {horizontalAlignment=HAlign_Center, verticalAlignment=VAlign_Center}}
  }
  Image_Line {
   widgetClassPath=&Image
   bIsVariable=true
   parent=VerticalBox_7
   widget=Image_Line
   properties {
    visibility=Hidden
    brush {
     bIsDynamicallyLoaded=false
     drawAs=Image
     imageType=NoImage
     mirroring=NoMirror
     resourceName=None
     resourceObject=null
     tiling=NoTile
     imageSize=<2 32>
     margin=<ltrb 0 0 0 0>
     outlineSettings {
      bUseBrushTransparency=false
      roundingType=HalfHeightRadius
      width=0
      color {colorUseRule=UseColor_Specified, specifiedColor=<rgba 0 0 0 0>}
      cornerRadii=<0 0 0 0>
     }
     tintColor {colorUseRule=UseColor_Specified, specifiedColor=<rgba 0.098 0.2235 0.302 0>}
     uVRegion {bIsValid=false, max=<0 0>, min=<0 0>}
    }
   }
   slot {class=VerticalBoxSlot, properties {horizontalAlignment=HAlign_Center}}
  }
  Overlay_0 {
   widgetClassPath=&Overlay
   parent=VerticalBox_7
   widget=Overlay_0
   slot {class=VerticalBoxSlot, properties {horizontalAlignment=HAlign_Center, verticalAlignment=VAlign_Center}}
  }
  SizeBox_0 {
   widgetClassPath=&SizeBox
   parent=Overlay_0
   widget=SizeBox_0
   properties {bOverride_WidthOverride=true, widthOverride=80}
   slot {class=OverlaySlot, properties {verticalAlignment=VAlign_Center}}
  }
  ProgressBar_Progress {
   widgetClassPath=/Script/UMG.ProgressBar
   bIsVariable=true
   parent=SizeBox_0
   widget=ProgressBar_Progress
   properties {
    renderTransform {angle=0, scale=<1 1>, shear=<0 0>, translation=<55 0>}
    widgetStyle {
     enableFillAnimation=false
     backgroundImage {
      bIsDynamicallyLoaded=false
      drawAs=Box
      imageType=FullColor
      mirroring=NoMirror
      resourceName=../../../Engine/Content/Slate/Common/ProgressBar_Background.png
      resourceObject=null
      tiling=NoTile
      imageSize=<0 0>
      margin=<ltrb 0.4167 0.4167 0.4167 0.4167>
      outlineSettings {
       bUseBrushTransparency=false
       roundingType=HalfHeightRadius
       width=0
       color {colorUseRule=UseColor_Specified, specifiedColor=<rgba 0 0 0 0>}
       cornerRadii=<0 0 0 0>
      }
      tintColor {colorUseRule=UseColor_Specified, specifiedColor=<rgba 0 0.1226 0.956 1>}
      uVRegion {bIsValid=false, max=<0 0>, min=<0 0>}
     }
     fillImage {
      bIsDynamicallyLoaded=false
      drawAs=Box
      imageType=FullColor
      mirroring=NoMirror
      resourceName=../../../Engine/Content/Slate/Common/ProgressBar_Fill.png
      resourceObject=null
      tiling=NoTile
      imageSize=<0 0>
      margin=<ltrb 0.4167 0.4167 0.4167 0.4167>
      outlineSettings {
       bUseBrushTransparency=false
       roundingType=HalfHeightRadius
       width=0
       color {colorUseRule=UseColor_Specified, specifiedColor=<rgba 0 0 0 0>}
       cornerRadii=<0 0 0 0>
      }
      tintColor {colorUseRule=UseColor_Specified, specifiedColor=<rgba 0 0.3771 0.8906 1>}
      uVRegion {bIsValid=false, max=<0 0>, min=<0 0>}
     }
     marqueeImage {
      bIsDynamicallyLoaded=false
      drawAs=Image
      imageType=FullColor
      mirroring=NoMirror
      resourceName=../../../Engine/Content/Slate/Common/ProgressBar_Marquee.png
      resourceObject=null
      tiling=Horizontal
      imageSize=<20 12>
      margin=<ltrb 0 0 0 0>
      outlineSettings {
       bUseBrushTransparency=false
       roundingType=HalfHeightRadius
       width=0
       color {colorUseRule=UseColor_Specified, specifiedColor=<rgba 0 0 0 0>}
       cornerRadii=<0 0 0 0>
      }
      tintColor {colorUseRule=UseColor_Specified, specifiedColor=<rgba 1 1 1 1>}
      uVRegion {bIsValid=false, max=<0 0>, min=<0 0>}
     }
    }
   }
   slot {class=SizeBoxSlot}
  }
  Widget_State {
   widgetClassPath=&/SurvivorPanelPrestigeState
   bIsVariable=true
   parent=Overlay_0
   widget=Widget_State
   slot {class=OverlaySlot, properties {horizontalAlignment=HAlign_Center, verticalAlignment=VAlign_Center}}
  }
  Text_Title {
   widgetClassPath=&CommonTextBlock
   bIsVariable=true
   parent=VerticalBox_7
   widget=Text_Title
   properties {
    style=/FortUILibrary/Styles/Typography/Foundation/TextStyle_BurbankSmall_XS_DarkCyan.TextStyle_BurbankSmall_XS_DarkCyan_C
    text=Title
    visibility=Collapsed
    colorAndOpacity {colorUseRule=UseColor_Specified, specifiedColor=<rgba 0.098 0.2235 0.302 1>}
    font {
     bForceMonospaced=false
     bMaterialIsStencil=false
     fontMaterial=null
     fontObject=&BurbankSmall
     letterSpacing=0
     monospacedWidth=1
     size=13
     skewAmount=0
     typefaceFontName=Black
     outlineSettings {
      bApplyOutlineToDropShadows=false
      bMiteredCorners=false
      bSeparateFillAlpha=false
      outlineMaterial=null
      outlineSize=0
      outlineColor=<rgba 0 0 0 1>
     }
    }
    shadowOffset=<0 0>
   }
   slot {class=VerticalBoxSlot, properties {horizontalAlignment=HAlign_Center, padding=<ltrb 8 10 10 0>}}
  }
 ]
}
