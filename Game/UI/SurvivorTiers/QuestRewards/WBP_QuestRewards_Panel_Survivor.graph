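# Quest rewards panel widget for the seasonal Survivor Medal, showing a tier badge, current prestige and tier, and a quest description.
# OnSurvivorTierStateChanged sets the text from display data and updates a row of prestige segments, Common through Mythic.
# Each segment's child index, not its own level property, is compared to current prestige: lower ones fill, higher ones empty, and the current one shows raw tier progress against max tier.
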WidgetBlueprint /Game/UI/SurvivorTiers/QuestRewards/WBP_QuestRewards_Panel_Survivor
graphs [
 EventGraph {nodes [
   n1 = Event OnSurvivorTierStateChanged!() -> (OutputDelegate, then, PrestigeLevel:int=0, TierRaw:int=0, TierDisplay:int=0) {is=title K2Node_Event}
   > UpdateCurrentState!(execute, self, Tier:int=@n1.TierDisplay, Prestige:int=@n1.PrestigeLevel) -> (then)
   > UpdateProgress!(execute, self, TierRaw:int=@n1.TierRaw, Prestige:int=@n1.PrestigeLevel) -> (then)
  ]}
 UpdateCurrentState {nodes [
   n1 = UpdateCurrentState!() -> (then, Tier:int, Prestige:int) {is=title K2Node_FunctionEntry}
   n2 = GetPrestigeData!(
    self:Athena Seasonal Survivor Tiers Display Data Object Reference=GetSurvivorTierDisplayData!() -> (ReturnValue:Athena Seasonal Survivor Tiers Display Data Object Reference)
    PrestigeLevel:int=@n1.Prestige
   ) -> (ReturnValue:Survivor Presige Data Structure)
   > "SetText (Text)"(
    self=Get Text_CurrentState() -> (Text_CurrentState:Common Text Object Reference)
    InText=Format Text!(
     Format="LOCTEXT(\"{Prestige} {Tier}\")"
     Prestige:Text=Break Survivor Presige Data!(SurvivorPresigeData:"Survivor Presige Data Structure (by ref)"=@n2) -> (DisplayName:Text, Color1:Linear Color Structure, Color2:Linear Color Structure, TextColor:Linear Color Structure) {is=title K2Node_BreakStruct}
     Tier:int=@n1
    ) -> (Result)
   )
  ]}
 UpdateProgress {nodes [
   n1 = UpdateProgress!() -> (then, TierRaw:int, Prestige:int) {is=title K2Node_FunctionEntry}
   > n2 = For Each Loop(Array:Array of Widget Object References=GetAllChildren(self=Get HB_Prestige() -> (HB_Prestige:Horizontal Box Object Reference))) -> (Array Element:Widget Object Reference)
   > n3 = Cast To SurvivorPanelPrestige!(execute, Object=@n2) -> (then, CastFailed, AsSurvivor Panel Prestige:Survivor Panel Prestige Object Reference) {is=title K2Node_DynamicCast}
   > n4 = Compare Int!(Exec:Exec, Input:int=@n2.Array Index, CompareWith:int=@n1.Prestige) -> (">":Exec, "==":Exec, "<":Exec) {is=title K2Node_MacroInstance}
   SetProgress!(execute=@n4."<", self:Survivor Panel Prestige Object Reference=@n3, PercentToNext:double=1, Achieved:bool=true) -> (then)
   n5 = int - int(A=GetMaxTier!(self:Athena Seasonal Survivor Tiers Display Data Object Reference=GetSurvivorTierDisplayData()) -> (ReturnValue:int=0))
   SetProgress(execute=@n4."==", self=@n3, PercentToNext=float / float(A="To Float (Integer)"(InInt=@n1), B="To Float (Integer)"(InInt=@n5)), Achieved=true)
   SetProgress(execute=@n4.">", self=@n3, PercentToNext=0, Achieved=false)
  ]}
]
variables []
widgets {
 info {
  inheritedWidgetCount=0
  parentClass=/Script/FortniteUI.FortAthenaSurvivorTiersQuestRewardsPanel
  rootWidgetClass=&SizeBox
  widgetCount=20
 }
 widgets [
  SizeBox_0 {widgetClassPath=&SizeBox, widget=SizeBox_0, properties {bOverride_WidthOverride=true, heightOverride=940, widthOverride=847}}
  Overlay_0 {widgetClassPath=&Overlay, parent=SizeBox_0, widget=Overlay_0, slot {class=SizeBoxSlot}}
  Image_Background {
   widgetClassPath=&Image
   bIsVariable=true
   parent=Overlay_0
   widget=Image_Background
   properties {
    brush {
     bIsDynamicallyLoaded=false
     drawAs=Image
     imageType=FullColor
     mirroring=NoMirror
     resourceName=None
     resourceObject=/Game/UI/SurvivorTiers/QuestRewards/Materials/M_UI_SurvivorTiers_Bg_Simple_FlexLegend
     tiling=NoTile
     imageSize=<1024 1024>
     margin=<ltrb 0 0 0 0>
     outlineSettings {
      bUseBrushTransparency=false
      roundingType=HalfHeightRadius
      width=0
      color {colorUseRule=UseColor_Specified, specifiedColor=<rgba 0 0 0 0>}
      cornerRadii=<0 0 0 0>
     }
     tintColor {colorUseRule=UseColor_Specified, specifiedColor=<rgba 1 1 1 1>}
     uVRegion {bIsValid=false, max=<0 0>, min=<0 0>}
    }
    renderTransform {angle=0, scale=<1.28 1.34>, shear=<0 0>, translation=<0 0>}
   }
   slot {class=OverlaySlot, properties {horizontalAlignment=HAlign_Fill, verticalAlignment=VAlign_Fill}}
  }
  VerticalBox_0 {
   widgetClassPath=&VerticalBox
   parent=Overlay_0
   widget=VerticalBox_0
   slot {class=OverlaySlot, properties {horizontalAlignment=HAlign_Fill, padding=<ltrb 0 80 0 0>}}
  }
  RichTextBlock_SelectedNewsTitle {
   widgetClassPath=&CommonRichTextBlock
   parent=VerticalBox_0
   widget=RichTextBlock_SelectedNewsTitle
   properties {
    autoWrapText=true
    bOverrideDefaultStyle=true
    defaultTextStyleOverrideClass=/FortUILibrary/Styles/Typography/_TextStyle_PrimaryH2._TextStyle_PrimaryH2_C
    justification=Center
    lineHeightPercentage=0.9
    scrollStyle=&TextScrollStyle-Base_C
    text=Survivor Medal
    textTransformPolicy=ToUpper
    wrapTextAt=1280
    defaultTextStyleOverride {
     overflowPolicy=Clip
     transformPolicy=None
     colorAndOpacity {colorUseRule=UseColor_Specified, specifiedColor=<rgba 1 1 1 1>}
     font {
      bForceMonospaced=false
      bMaterialIsStencil=false
      fontMaterial=null
      fontObject=&Burbank
      letterSpacing=0
      monospacedWidth=1
      size=64
      skewAmount=0
      typefaceFontName=Burbank-Black
      outlineSettings {
       bApplyOutlineToDropShadows=false
       bMiteredCorners=false
       bSeparateFillAlpha=false
       outlineMaterial=null
       outlineSize=0
       outlineColor=<rgba 0 0 0 1>
      }
     }
     highlightColor {colorUseRule=UseColor_Specified, specifiedColor=<rgba 0 0 0 1>}
     highlightShape {
      bIsDynamicallyLoaded=false
      drawAs=Image
      imageType=NoImage
      mirroring=NoMirror
      resourceName=None
      resourceObject=null
      tiling=NoTile
      imageSize=<32 32>
      margin=<ltrb 0 0 0 0>
      outlineSettings {
       bUseBrushTransparency=false
       roundingType=HalfHeightRadius
       width=0
       color {colorUseRule=UseColor_Specified, specifiedColor=<rgba 0 0 0 0>}
       cornerRadii=<0 0 0 0>
      }
      tintColor {colorUseRule=UseColor_Specified, specifiedColor=<rgba 1 1 1 1>}
      uVRegion {bIsValid=false, max=<0 0>, min=<0 0>}
     }
     selectedBackgroundColor {colorUseRule=UseColor_Foreground, specifiedColor=<rgba 1 0 1 1>}
     shadowColorAndOpacity=<rgba 0 0 0 1>
     shadowOffset=<0 0>
     strikeBrush {
      bIsDynamicallyLoaded=false
      drawAs=Image
      imageType=NoImage
      mirroring=NoMirror
      resourceName=None
      resourceObject=null
      tiling=NoTile
      imageSize=<32 32>
      margin=<ltrb 0 0 0 0>
      outlineSettings {
       bUseBrushTransparency=false
       roundingType=HalfHeightRadius
       width=0
       color {colorUseRule=UseColor_Specified, specifiedColor=<rgba 0 0 0 0>}
       cornerRadii=<0 0 0 0>
      }
      tintColor {colorUseRule=UseColor_Specified, specifiedColor=<rgba 1 1 1 1>}
      uVRegion {bIsValid=false, max=<0 0>, min=<0 0>}
     }
     underlineBrush {
      bIsDynamicallyLoaded=false
      drawAs=Image
      imageType=NoImage
      mirroring=NoMirror
      resourceName=None
      resourceObject=null
      tiling=NoTile
      imageSize=<32 32>
      margin=<ltrb 0 0 0 0>
      outlineSettings {
       bUseBrushTransparency=false
       roundingType=HalfHeightRadius
       width=0
       color {colorUseRule=UseColor_Specified, specifiedColor=<rgba 0 0 0 0>}
       cornerRadii=<0 0 0 0>
      }
      tintColor {colorUseRule=UseColor_Specified, specifiedColor=<rgba 1 1 1 1>}
      uVRegion {bIsValid=false, max=<0 0>, min=<0 0>}
     }
    }
    margin=<ltrb 0 0 0 16>
   }
   slot {class=VerticalBoxSlot}
  }
  HorizontalBox_81 {widgetClassPath=&HorizontalBox, parent=VerticalBox_0, widget=HorizontalBox_81, slot {class=VerticalBoxSlot}}
  Text_CurrentState {
   widgetClassPath=&CommonTextBlock
   bIsVariable=true
   parent=HorizontalBox_81
   widget=Text_CurrentState
   properties {
    style=/FortUILibrary/Styles/Typography/_TextStyle_SubH4._TextStyle_SubH4_C
    text="Current Prestige && Tier"
    font {
     bForceMonospaced=false
     bMaterialIsStencil=false
     fontMaterial=null
     fontObject=&BurbankSmall
     letterSpacing=0
     monospacedWidth=1
     size=24
     skewAmount=0
     typefaceFontName=Black
     outlineSettings {
      bApplyOutlineToDropShadows=false
      bMiteredCorners=false
      bSeparateFillAlpha=false
      outlineMaterial=null
      outlineSize=0
      outlineColor=<rgba 0 0 0 1>
     }
    }
    shadowOffset=<0 0>
    strikeBrush {
     bIsDynamicallyLoaded=false
     drawAs=NoDrawType
     imageType=NoImage
     mirroring=NoMirror
     resourceName=None
     resourceObject=null
     tiling=NoTile
     imageSize=<32 32>
     margin=<ltrb 0 0 0 0>
     outlineSettings {
      bUseBrushTransparency=false
      roundingType=HalfHeightRadius
      width=0
      color {colorUseRule=UseColor_Specified, specifiedColor=<rgba 0 0 0 0>}
      cornerRadii=<0 0 0 0>
     }
     tintColor {colorUseRule=UseColor_Specified, specifiedColor=<rgba 1 1 1 1>}
     uVRegion {bIsValid=false, max=<0 0>, min=<0 0>}
    }
   }
   slot {class=HorizontalBoxSlot, properties {horizontalAlignment=HAlign_Center, padding=<ltrb 10 0 10 10>, size {sizeRule=Fill, value=1}}}
  }
  W_SurvivorTierBadge {
   widgetClassPath=/Game/UI/SurvivorTiers/W_SurvivorTierBadge
   bIsVariable=true
   parent=VerticalBox_0
   widget=W_SurvivorTierBadge
   properties {badgeSize=Large, editorPreviewPrestigeLevel=0, renderTransform {angle=0, scale=<0.9 0.9>, shear=<0 0>, translation=<0 0>}}
   slot {class=VerticalBoxSlot, properties {horizontalAlignment=HAlign_Center, verticalAlignment=VAlign_Top}}
  }
  Image_ForegroundGlow {
   widgetClassPath=&Image
   bIsVariable=true
   parent=Overlay_0
   widget=Image_ForegroundGlow
   properties {
    brush {
     bIsDynamicallyLoaded=false
     drawAs=NoDrawType
     imageType=NoImage
     mirroring=NoMirror
     resourceName=None
     resourceObject=null
     tiling=NoTile
     imageSize=<1024 1024>
     margin=<ltrb 0 0 0 0>
     outlineSettings {
      bUseBrushTransparency=false
      roundingType=HalfHeightRadius
      width=0
      color {colorUseRule=UseColor_Specified, specifiedColor=<rgba 0 0 0 0>}
      cornerRadii=<0 0 0 0>
     }
     tintColor {colorUseRule=UseColor_Specified, specifiedColor=<rgba 1 1 1 1>}
     uVRegion {bIsValid=false, max=<0 0>, min=<0 0>}
    }
    renderTransform {angle=0, scale=<1.25 1.25>, shear=<0 0>, translation=<0 -83>}
   }
   slot {class=OverlaySlot, properties {horizontalAlignment=HAlign_Right, verticalAlignment=VAlign_Center, padding=<ltrb 0 0 -50 0>}}
  }
  VerticalBox_1 {
   widgetClassPath=&VerticalBox
   parent=Overlay_0
   widget=VerticalBox_1
   slot {class=OverlaySlot, properties {horizontalAlignment=HAlign_Fill, verticalAlignment=VAlign_Bottom, padding=<ltrb 40 40 40 40>}}
  }
  HB_Prestige {
   widgetClassPath=&HorizontalBox
   bIsVariable=true
   parent=VerticalBox_1
   widget=HB_Prestige
   properties {renderTransform {angle=0, scale=<1 1>, shear=<0 0>, translation=<0 5>}}
   slot {class=VerticalBoxSlot, properties {verticalAlignment=VAlign_Top}}
  }
  Common {
   widgetClassPath=&/SurvivorPanelPrestige
   parent=HB_Prestige
   widget=Common
   properties {lineHeight=20}
   slot {class=HorizontalBoxSlot, properties {verticalAlignment=VAlign_Bottom, size {sizeRule=Fill, value=1}}}
  }
  Uncommon {
   widgetClassPath=&SurvivorPanelPrestige
   parent=HB_Prestige
   widget=Uncommon
   properties {lineHeight=30, presigeLevel=1}
   slot {class=HorizontalBoxSlot, properties {verticalAlignment=VAlign_Bottom, size {sizeRule=Fill, value=1}}}
  }
  Rare {
   widgetClassPath=&SurvivorPanelPrestige
   parent=HB_Prestige
   widget=Rare
   properties {lineHeight=40, presigeLevel=2}
   slot {class=HorizontalBoxSlot, properties {verticalAlignment=VAlign_Bottom, size {sizeRule=Fill, value=1}}}
  }
  Epic {
   widgetClassPath=&SurvivorPanelPrestige
   parent=HB_Prestige
   widget=Epic
   properties {lineHeight=50, presigeLevel=3}
   slot {class=HorizontalBoxSlot, properties {verticalAlignment=VAlign_Bottom, size {sizeRule=Fill, value=1}}}
  }
  Legendary {
   widgetClassPath=&SurvivorPanelPrestige
   parent=HB_Prestige
   widget=Legendary
   properties {lineHeight=60, presigeLevel=4}
   slot {class=HorizontalBoxSlot, properties {verticalAlignment=VAlign_Bottom, size {sizeRule=Fill, value=1}}}
  }
  Exotic {
   widgetClassPath=&SurvivorPanelPrestige
   parent=HB_Prestige
   widget=Exotic
   properties {lineHeight=70, presigeLevel=5}
   slot {class=HorizontalBoxSlot, properties {verticalAlignment=VAlign_Bottom, size {sizeRule=Fill, value=1}}}
  }
  Mythic {
   widgetClassPath=&SurvivorPanelPrestige
   parent=HB_Prestige
   widget=Mythic
   properties {lineHeight=80, presigeLevel=6, showProgress=false}
   slot {class=HorizontalBoxSlot, properties {verticalAlignment=VAlign_Bottom, size {sizeRule=Fill, value=1}}}
  }
  Border {
   widgetClassPath=&Border
   parent=VerticalBox_1
   widget=Border
   properties {
    horizontalAlignment=HAlign_Center
    background {
     bIsDynamicallyLoaded=false
     drawAs=Image
     imageType=NoImage
     mirroring=NoMirror
     resourceName=None
     resourceObject=/Game/UI/SurvivorTiers/QuestRewards/Materials/MI_UI_QuestRewards_InfoBackground_SurvivorTiers
     tiling=NoTile
     imageSize=<32 32>
     margin=<ltrb 0 0 0 0>
     outlineSettings {
      bUseBrushTransparency=false
      roundingType=HalfHeightRadius
      width=0
      color {colorUseRule=UseColor_Specified, specifiedColor=<rgba 0 0 0 0>}
      cornerRadii=<0 0 0 0>
     }
     tintColor {colorUseRule=UseColor_Specified, specifiedColor=<rgba 1 1 1 1>}
     uVRegion {bIsValid=false, max=<0 0>, min=<0 0>}
    }
    brushColor=<rgba 1 1 1 100>
    padding=<ltrb 40 20 40 20>
   }
   slot {class=VerticalBoxSlot, properties {padding=<ltrb 0 20 0 0>}}
  }
  RichText_Desc {
   widgetClassPath=&CommonRichTextBlock
   parent=Border
   widget=RichText_Desc
   properties {
    autoWrapText=true
    justification=Center
    text="Complete <Highlight>Quests</> to upgrade your Survivor Medal. Earn the highest Medal before the end of the season.\r\n\r\nWin the ultimate Survivor Medal by reaching Mythic 1."
    textStyleSet=&/DT_UI_RichTextStyles_QuestRewards_Survivor
    defaultTextStyleOverride {
     overflowPolicy=Clip
     transformPolicy=None
     colorAndOpacity {colorUseRule=UseColor_Specified, specifiedColor=<rgba 1 1 1 1>}
     font {
      bForceMonospaced=false
      bMaterialIsStencil=false
      fontMaterial=null
      fontObject=&BurbankSmall
      letterSpacing=0
      monospacedWidth=1
      size=19
      skewAmount=0
      typefaceFontName=Bold
      outlineSettings {
       bApplyOutlineToDropShadows=false
       bMiteredCorners=false
       bSeparateFillAlpha=false
       outlineMaterial=null
       outlineSize=0
       outlineColor=<rgba 0 0 0 1>
      }
     }
     highlightColor {colorUseRule=UseColor_Specified, specifiedColor=<rgba 0 0 0 1>}
     highlightShape {
      bIsDynamicallyLoaded=false
      drawAs=Image
      imageType=NoImage
      mirroring=NoMirror
      resourceName=None
      resourceObject=null
      tiling=NoTile
      imageSize=<32 32>
      margin=<ltrb 0 0 0 0>
      outlineSettings {
       bUseBrushTransparency=false
       roundingType=HalfHeightRadius
       width=0
       color {colorUseRule=UseColor_Specified, specifiedColor=<rgba 0 0 0 0>}
       cornerRadii=<0 0 0 0>
      }
      tintColor {colorUseRule=UseColor_Specified, specifiedColor=<rgba 1 1 1 1>}
      uVRegion {bIsValid=false, max=<0 0>, min=<0 0>}
     }
     selectedBackgroundColor {colorUseRule=UseColor_Foreground, specifiedColor=<rgba 1 0 1 1>}
     shadowColorAndOpacity=<rgba 0 0 0 1>
     shadowOffset=<0 0>
     strikeBrush {
      bIsDynamicallyLoaded=false
      drawAs=Image
      imageType=NoImage
      mirroring=NoMirror
      resourceName=None
      resourceObject=null
      tiling=NoTile
      imageSize=<32 32>
      margin=<ltrb 0 0 0 0>
      outlineSettings {
       bUseBrushTransparency=false
       roundingType=HalfHeightRadius
       width=0
       color {colorUseRule=UseColor_Specified, specifiedColor=<rgba 0 0 0 0>}
       cornerRadii=<0 0 0 0>
      }
      tintColor {colorUseRule=UseColor_Specified, specifiedColor=<rgba 1 1 1 1>}
      uVRegion {bIsValid=false, max=<0 0>, min=<0 0>}
     }
     underlineBrush {
      bIsDynamicallyLoaded=false
      drawAs=Image
      imageType=NoImage
      mirroring=NoMirror
      resourceName=None
      resourceObject=null
      tiling=NoTile
      imageSize=<32 32>
      margin=<ltrb 0 0 0 0>
      outlineSettings {
       bUseBrushTransparency=false
       roundingType=HalfHeightRadius
       width=0
       color {colorUseRule=UseColor_Specified, specifiedColor=<rgba 0 0 0 0>}
       cornerRadii=<0 0 0 0>
      }
      tintColor {colorUseRule=UseColor_Specified, specifiedColor=<rgba 1 1 1 1>}
      uVRegion {bIsValid=false, max=<0 0>, min=<0 0>}
     }
    }
   }
   slot {class=BorderSlot, properties {horizontalAlignment=HAlign_Center, padding=<ltrb 40 20 40 20>}}
  }
 ]
}
